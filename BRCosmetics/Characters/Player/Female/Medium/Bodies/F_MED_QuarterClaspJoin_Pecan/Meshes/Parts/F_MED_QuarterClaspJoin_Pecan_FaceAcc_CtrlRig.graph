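# Control Rig for the face accessory part of the F_MED_QuarterClaspJoin_Pecan female body, posing one dynamic hair strand that hangs from the head toward the left shoulder and upper spine.
# A single sequence branch runs chained weighted parent constraints on dyn_hair_parent through dyn_hair_05.
# The root follows the head, and weights shift down the strand to blends of left clavicle and upper spine bones.
# The other hair chains in the hierarchy are not driven.
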ControlRigBlueprint /BRCosmetics/Characters/Player/Female/Medium/Bodies/F_MED_QuarterClaspJoin_Pecan/Meshes/Parts/F_MED_QuarterClaspJoin_Pecan_FaceAcc_CtrlRig
graphs [
 RigVMFunctionLibrary {links [], nodes []}
 RigVMModel {nodes [
   RigUnit_BeginExecution = &RigUnit_BeginExecution()
   RigVMFunction_Sequence = &RigVMFunction_Sequence(ExecuteContext=@RigUnit_BeginExecution.ExecutePin)
   &RigUnit_ParentConstraint(
    AdvancedSettings=(RotationOrderForFilter=XZY)
    Child=(Type=Bone,Name="dyn_hair_parent")
    Parents=((Item=(Type=Bone,Name="head"),Weight=1.0))
    Weight=1
    ExecutePin=@RigVMFunction_Sequence.A
   )
   > &RigUnit_ParentConstraint(
    AdvancedSettings=(RotationOrderForFilter=XZY)
    Child=(Type=Bone,Name="dyn_hair_01")
    Parents=((Item=(Type=Bone,Name="head"),Weight=0.5),(Item=(Type=Bone,Name="clavicle_l"),Weight=0.1),(Item=(Type=Bone,Name="spine_05"),Weight=0.4))
    Weight=1
   )
   > &RigUnit_ParentConstraint(
    AdvancedSettings=(RotationOrderForFilter=XZY)
    Child=(Type=Bone,Name="dyn_hair_02")
    Parents=((Item=(Type=Bone,Name="head"),Weight=0.25),(Item=(Type=Bone,Name="clavicle_l"),Weight=0.25),(Item=(Type=Bone,Name="spine_05"),Weight=0.5))
    Weight=1
   )
   > &RigUnit_ParentConstraint(
    AdvancedSettings=(RotationOrderForFilter=XZY)
    Child=(Type=Bone,Name="dyn_hair_03")
    Parents=((Item=(Type=Bone,Name="spine_04"),Weight=0.0669),(Item=(Type=Bone,Name="clavicle_l"),Weight=0.1177),(Item=(Type=Bone,Name="spine_05"),Weight=0.8154))
    Weight=1
   )
   > &RigUnit_ParentConstraint(
    AdvancedSettings=(RotationOrderForFilter=XZY)
    Child=(Type=Bone,Name="dyn_hair_04")
    Parents=((Item=(Type=Bone,Name="spine_05"),Weight=0.3449),(Item=(Type=Bone,Name="spine_04"),Weight=0.6551))
    Weight=1
   )
   > &RigUnit_ParentConstraint(
    AdvancedSettings=(RotationOrderForFilter=XZY)
    Child=(Type=Bone,Name="dyn_hair_05")
    Parents=((Item=(Type=Bone,Name="spine_05"),Weight=0.186),(Item=(Type=Bone,Name="spine_04"),Weight=0.814))
    Weight=1
   )
  ]}
]
hierarchy [
 root {loc=<0 0 0>, rot=<0 0 0 1>}
 pelvis {parent=root, loc=<0 -0.5301 98.0571>, rot=<0.0127 0.707 -0.0127 -0.707>}
 spine_01 {parent=pelvis, loc=<6.1208 -0.9328 0>, rot=<0 0 0.034 -0.9994>}
 spine_02 {parent=spine_01, loc=<6.5997 -0.2095 0>, rot=<0 0 0.0238 -0.9997>}
 spine_03 {parent=spine_02, loc=<6.4914 -0.1237 0>, rot=<0 0 0.0149 -0.9999>}
 spine_04 {parent=spine_03, loc=<6.9598 0.0794 0>, rot=<0 0 -0.0132 -0.9999>}
 spine_05 {parent=spine_04, loc=<6.9987 0 0>, rot=<0 0 -0.0015 -1>}
 neck_01 {parent=spine_05, loc=<11.3138 0.1529 0>, rot=<0 0 -0.2003 -0.9797>}
 neck_02 {parent=neck_01, loc=<4.5627 0 0>, rot=<0 0 0.0504 -0.9987>}
 head {parent=neck_02, loc=<5.3914 0 0>, rot=<0 0 0.115 -0.9934>}
 dyn_hair_bk_r_base {parent=head, loc=<2.6263 -3.853 5.867>, rot=<0.0743 -0.424 0.8922 -0.1369>}
 dyn_hair_bk_r_parent {parent=dyn_hair_bk_r_base, loc=<0 0 0>, rot=<0 0 0 1>}
 dyn_hair_bk_r_01 {parent=dyn_hair_bk_r_parent, loc=<6.0461 -0.5909 -0.864>, rot=<0.1362 0.0809 -0.1117 -0.981>}
 dyn_hair_bk_r_02 {parent=dyn_hair_bk_r_01, loc=<6.9335 0 0>, rot=<-0.0698 0 0 -0.9976>}
 dyn_hair_bk_mid_base {parent=head, loc=<2.2224 -6.1632 2.7259>, rot=<0.0719 -0.579 0.8121 -0.0088>}
 dyn_hair_bk_mid_parent {parent=dyn_hair_bk_mid_base, loc=<0 0 0>, rot=<0 0 0 1>}
 dyn_hair_bk_mid_01 {parent=dyn_hair_bk_mid_parent, loc=<5.5169 0 0>, rot=<0.258 0.2147 -0.1263 -0.9335>}
 dyn_hair_bk_02 {parent=dyn_hair_bk_mid_01, loc=<6.0179 0 0>, rot=<-0.2172 -0.0977 0.1049 -0.9655>}
 dyn_hair_bk_l_base {parent=head, loc=<2.2538 -6.2987 -1.8152>, rot=<0.0032 -0.766 0.6428 -0.0027>}
 dyn_hair_bk_l_parent {parent=dyn_hair_bk_l_base, loc=<0 0 0>, rot=<0 0 0 1>}
 dyn_hair_bk_l_01 {parent=dyn_hair_bk_l_parent, loc=<5.5131 0.5016 1.1859>, rot=<0.1017 0.2417 -0.0255 -0.9647>}
 dyn_hair_bk_l_02 {parent=dyn_hair_bk_l_01, loc=<7.0897 0 0>, rot=<0 0 0 1>}
 dyn_hair_mid_l_base {parent=head, loc=<2.6436 -3.6767 -6.2618>, rot=<0.0893 -0.9397 0.3068 -0.1216>}
 dyn_hair_mid_l_parent {parent=dyn_hair_mid_l_base, loc=<0 0 0>, rot=<0 0 0 1>}
 dyn_hair_mid_l_01 {parent=dyn_hair_mid_l_parent, loc=<5.4213 0 0>, rot=<-0.0387 0.0027 -0.0241 -0.999>}
 dyn_hair_mid_l_02 {parent=dyn_hair_mid_l_01, loc=<3.8324 0 0>, rot=<0.0523 0 0 -0.9986>}
 dyn_hair_base {parent=head, loc=<2.52 5.7462 -6.4959>, rot=<0.0304 -0.9611 0.0073 -0.2744>}
 dyn_hair_parent {parent=dyn_hair_base, loc=<0 0 0>, rot=<0 0 0 1>}
 dyn_hair_01 {parent=dyn_hair_parent, loc=<4.7947 -0.2205 -0.8879>, rot=<0.0728 -0.267 0.0654 -0.9587>}
 dyn_hair_02 {parent=dyn_hair_01, loc=<4.8923 0 0>, rot=<0.0845 -0.2168 -0.2382 -0.943>}
 dyn_hair_03 {parent=dyn_hair_02, loc=<5.5595 0 0>, rot=<0.1579 0.2951 -0.123 -0.9343>}
 dyn_hair_04 {parent=dyn_hair_03, loc=<6.6748 0 0>, rot=<-0.1626 0.1218 0.2021 -0.9581>}
 dyn_hair_05 {parent=dyn_hair_04, loc=<3.7419 0 0>, rot=<-0.1509 -0.0248 -0.065 -0.9861>}
 dyn_hair_06 {parent=dyn_hair_05, loc=<4.1463 -2.3878 -0.6041>, rot=<0.173 -0.0436 0.0755 -0.9811>}
 dyn_hair_07 {parent=dyn_hair_05, loc=<4.2014 -2.3509 -4.5793>, rot=<0.1779 -0.0473 0.1256 -0.9749>}
 dyn_hair_08 {parent=dyn_hair_05, loc=<3.7845 -2.4255 1.9131>, rot=<0.1062 0.1978 0.239 -0.9447>}
 clavicle_l {parent=spine_05, loc=<5.4831 -0.1053 -3.2946>, rot=<0.0465 -0.7289 -0.009 -0.6829>}
 upperarm_l {parent=clavicle_l, loc=<9.8658 -1.2136 -0.4105>, rot=<0.0164 -0.3245 -0.0269 -0.9454>}
 lowerarm_l {parent=upperarm_l, loc=<26.465 -0.8345 0.259>, rot=<0.0036 0.0161 -0.2298 -0.9731>}
 clavicle_r {parent=spine_05, loc=<5.4831 -0.1053 3.2946>, rot=<0.7289 0.0465 -0.6829 0.009>}
 upperarm_r {parent=clavicle_r, loc=<-9.8658 1.2136 0.4105>, rot=<0.0164 -0.3245 -0.0269 -0.9454>}
 lowerarm_r {parent=upperarm_r, loc=<-26.465 0.8345 -0.259>, rot=<0.0036 0.0161 -0.2298 -0.9731>}
]
